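# Unlit masked pearl effect material for Niagara mesh and sprite particles in the Starbeam set, apparently temporary by its name.
# A Fresnel-driven two-color gradient blends against a channel-masked cloud noise path via a color/texture scalar.
# The result feeds an HDR power boost and a desaturated particle-color term, summed into MF_FX_EmissiveAdjustments.
# Dynamic parameters drive desaturation and texture-path Fresnel intensity; Metallic, Specular and Roughness parameters are leftovers feeding only dangling reroutes.
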Material /DealershipSharedLibrary/Assets/Sets/Starbeam/Materials/M_Pearl_mesh_Temp
graph {
 properties {
  bUsedWithNanite=true
  bUsedWithNiagaraMeshParticles=true
  bUsedWithNiagaraSprites=true
  blendMode=BLEND_Masked
  pixelDepthOffsetMode=PDOM_Legacy
  shadingModel=MSM_Unlit
  usedShadingModels=Unlit
 }
 nodes [
  n1 = Fresnel() {exponent=0.5}
  n2 = ScalarParameter() {defaultValue=1, parameterName=FresPower}
  n3 = Clamp(
   None=Multiply(A=Power(Base=@n1, Exp=@n2) {constExponent=1}, B=ScalarParameter() {defaultValue=1.25, parameterName=FresReverse?}) {constB=1.25}
   Max=ScalarParameter() {defaultValue=1, parameterName=FresClampMin?}
  )
  n4 = ScalarParameter() {defaultValue=1, parameterName="Color-\"1\"/Texture-\"0\""}
  n5 = LinearInterpolate(
   A=NamedRerouteUsage()
   B=LinearInterpolate(
    A=VectorParameter() {parameterName=Inner Color, defaultValue=<rgba 0.2852 0.0127 1.5 1>}.RGB
    B=VectorParameter() {parameterName=Outer Color, defaultValue=<rgba 0 0.1798 0.25 1>}.RGB
    Alpha=@n3
   )
   Alpha=@n4
  )
  n6 = Multiply(
   A=Power(Base=@n5, Exp=ScalarParameter() {defaultValue=10, parameterName=HDRPower}) {constExponent=10}
   B=ScalarParameter() {defaultValue=3, parameterName=HDRMultiply}
  ) {constB=3}
  n7 = Multiply(A=@n4, B=Multiply(A=@n5, B=ParticleColor().RGB))
  n8 = DynamicParameter() {defaultValue=<rgba 1 0 1 1>, paramNames [FresInt,Desaturation,Param3,Param4]}
  MaterialFunctionCall(Emissive=Add(A=Multiply(A=@n6, B=@n3), B=Desaturation(None=@n7, Fraction=@n8.Desaturation))) {materialFunction=&MF_FX_EmissiveAdjustments}
  n9 = DynamicParameter() {defaultValue=<rgba 1 0 1 1>, paramNames [FresInt,Desaturation,Param3,Param4]}
  n10 = Multiply(A=Fresnel() {exponent=3}, B=@n9.FresInt)
  n11 = TextureSampleParameter2D(UVs=TextureCoordinate()) {parameterName=Texture, samplerType=SAMPLERTYPE_LinearColor, texture=/Game/Global/Textures/Default/Noise/T_CloudNoise_01}
  n12 = Desaturation(
   None=Multiply(A=ParticleColor().RGB, B=ChannelMaskParameter(None=@n11.RGBA) {maskChannel=Green, parameterName=Texture_Mask, defaultValue=<rgba 0 1 0 0>}.R)
   Fraction=@n9.Desaturation
  )
  NamedRerouteDeclaration(None=Add(A=Add(A=@n10, B=@n12), B=@n12)) {name=Texture, nodeColor=<rgba 0.4297 0 1 1>}
  Reroute(None=Multiply(A=@n5) {constB=0.3})
  Reroute(None=ScalarParameter() {parameterName=Metallic})
  Reroute(None=ScalarParameter() {defaultValue=1, parameterName=Specular})
  Reroute(None=ScalarParameter() {defaultValue=0.2, parameterName=Roughness})
  ScalarParameter() {parameterName=DesaturationParticleColor}
  ScalarParameter() {parameterName=DesaturationParticleColor}
  VectorParameter() {defaultValue=<rgba 0 0 0 1>}
  VectorParameter() {parameterName=Param_1, defaultValue=<rgba 1 1 1 1>}
 ]
}
